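# Niagara swing trail effect for the female Wicked Duck pickaxe cosmetic, drawing short-lived, warm-tinted ribbons while the weapon swings.
# A single Trail emitter spawns two AnimTrails between bone or socket pairs, using the shared AnimTrail modules.
# Negated owner velocity feeds a travelled-distance calculation that sets ribbon UV distance at spawn.
# Curves narrow each ribbon and fade alpha, combining an age fade with a normalized-particle-count ramp along the trail.
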NiagaraSystem /BRCosmetics/Weapons/FORT_Melee/Pickaxe_Wicked_Duck_Female/FX/NS_Wicked_Duck_FemaleSwing
dependencies {usedDataInterfaces [&NiagaraDataInterfaceCurve], usedDynamicInputs [], usedModules [&SystemState]}
emitters {Trail {
  emitterUpdateScript [
   &EmitterState(Life Cycle Mode=::NewEnumerator1, Loop Behavior=::NewEnumerator1, Loop Duration=0.2, UseLoopDelay=false) {module=EmitterState001}
   &FN_SpawnAnimTrail_v2(Number of Anim Trails=2)
   &self:CalcDistance_Travelled(Velocity Vector:Vector3f=&Multiply_Vector(A=$Engine.Owner.Velocity, B=&VectorFromFloat(Value=-1))) {module=CalcDistance_Travelled}
  ]
  particleSpawnScript [
   /Game/Effects/Niagara/Modules/AnimTrail/FN_InitializeAnimTrail/FN_InitializeAnimTrailBoneIndecies(First Bone or Socket Index=1)
   &FN_InitializeAnimTrailBoneIndecies(AnimTrail Selection=1, First Bone or Socket Index=3, Second Bone or Socket Index=4) {module=FN_InitializeAnimTrailBoneIndecies001}
   /Game/Effects/Niagara/Modules/AnimTrail/FN_InitializeAnimTrail/FN_InitializeAnimTrail(AnimTrailWidth=1.5, Color=<rgba 1 0.3355 0.125 0.5>, Lifetime=0.2)
   /SetVariables(Particles.RibbonUVDistance:NiagaraFloat=$Emitter.DistanceTraveled) {isSetParameters=true}
  ]
  particleUpdateScript [
   &FN_UpdateAnimTrail_v2()
   &ScaleRibbonWidth(Ribbon Width Scale=&FloatFromCurve(FloatCurve=DataInterface() {propertyValues {Curve {keys [
         {
          arriveTangent=0.5623
          arriveTangentWeight=0.4591
          interpMode=RCIM_Cubic
          leaveTangent=0.5623
          leaveTangentWeight=0.4591
          tangentMode=RCTM_User
          tangentWeightMode=RCTWM_WeightedBoth
          time=0
          value=1
         }
         {
          arriveTangent=-0.2443
          arriveTangentWeight=0.7082
          interpMode=RCIM_Cubic
          leaveTangent=-0.2443
          leaveTangentWeight=0.7082
          tangentMode=RCTM_User
          tangentWeightMode=RCTWM_WeightedBoth
          time=1
          value=0
         }
        ]}}}))
   &ScaleColor(
    Scale Alpha=&Multiply_Float(
     A=&FloatFromCurve(FloatCurve=DataInterface() {propertyValues {Curve {keys [
          {
           arriveTangent=-3.175
           arriveTangentWeight=0.9161
           interpMode=RCIM_Cubic
           leaveTangent=-3.175
           leaveTangentWeight=0.9161
           tangentMode=RCTM_User
           tangentWeightMode=RCTWM_WeightedBoth
           time=0
           value=1
          }
          {
           arriveTangent=-0.0544
           arriveTangentWeight=0.581
           interpMode=RCIM_Cubic
           leaveTangent=-0.3338
           leaveTangentWeight=0.2826
           tangentMode=RCTM_Break
           tangentWeightMode=RCTWM_WeightedBoth
           time=1
           value=0
          }
         ]}}})
     B=&FloatFromCurve(
      FloatCurve=DataInterface() {propertyValues {Curve {keys [
          {
           interpMode=RCIM_Cubic
           leaveTangent=6.4657
           leaveTangentWeight=1.1441
           tangentMode=RCTM_Break
           tangentWeightMode=RCTWM_WeightedBoth
           time=0
           value=0
          }
          {interpMode=RCIM_Cubic, tangentMode=RCTM_Break, time=1, value=1}
         ]}}}
      CurveIndex=/Game/Effects/Niagara/DynamicInputs/Math/FN_NormalizedParticleCount()
      Scale Curve=2
     )
    )
    ScaleA=true
    ScaleRGB=false
   )
  ]
  rendererClasses [&NiagaraRibbonRendererProperties]
  renderers [{rendererClass=&NiagaraRibbonRendererProperties}]
 }}
summary {
 systemName=NS_Wicked_Duck_FemaleSwing
 emitters [{emitterName=Trail, rendererClasses [&NiagaraRibbonRendererProperties]}]
 userVariables [
  User.RibbonIDOffset {type=NiagaraInt32}
  User.SkeletalMesh {
   type=NiagaraDataInterfaceSkeletalMesh
   defaultValue {
    struct=NiagaraExt_VariableValue_DataInterface
    value {dataInterface=&self:NiagaraDataInterfaceSkeletalMesh_0}
   }
  }
 ]
}
