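# Anim Blueprint for the Wicked Duck Female pickaxe, adding springy coil motion and a hammerhead squash on hit.
# Five chained AnimDynamics linear springs, relative to the handle bone, slide the presser and coil joints along X with graduated travel limits.
# The cached coil pose feeds a Modify Bone that flattens the right hammer spring joint's X scale.
# The update event raises that alpha briefly on SuccessfulHit; the Stretch alpha drives nothing in the AnimGraph.
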AnimBlueprint /BRCosmetics/Weapons/FORT_Melee/Pickaxe_Wicked_Duck_Female/Meshes/Wicked_Duck_Female_AnimBP
target_skeleton=&/Wicked_Duck_Female_Skeleton
graphs [
 AnimGraph {nodes [
   Output Pose(Result=Use cached pose 'HammerSquash'()) {layerGroup=None}
   AnimGraphNode_SubInput_0 = Input Pose()
   n1 = Anim Dynamics - Bone: dyn_springPresser_jnt(ComponentPose=Local To Component(LocalPose=@AnimGraphNode_SubInput_0), Alpha=1, ExternalForce="0,0,0") {
    bGravityOverrideInSimSpace=true
    bLinearSpring=true
    bOverrideLinearDamping=true
    gravityScale=0.4
    linearDampingOverride=0.9999
    linearSpringConstant=300
    simulationSpace=BoneRelative
    alphaBoolBlend {blendOption=Linear}
    boundBone {boneName=dyn_springPresser_jnt}
    componentAppliedLinearAccClamp=<2000 2000 2000>
    componentLinearAccScale=<0.4 0.4 0.4>
    componentLinearVelScale=<0.4 0.4 0.4>
    physicsBodyDefinitions [{
      sphereCollisionRadius=0
      boundBone {boneName=dyn_springPresser_jnt}
      boxExtents=<2 2 2>
      constraintSetup {angularLimitsMax=<0 0 0>, angularLimitsMin=<0 0 0>, angularTarget=<0 0 0>, linearAxesMax=<5.3 0 0>}
      localJointOffset=<0 0 0>
     }]
    relativeSpaceBone {boneName=Melee_TwoHanded_Handle}
   }
   n2 = Anim Dynamics - Bone: dyn_spring_jnt_1(ComponentPose=@n1, Alpha=1, ExternalForce="0,0,0") {
    bGravityOverrideInSimSpace=true
    bLinearSpring=true
    bOverrideLinearDamping=true
    gravityScale=0.4
    linearDampingOverride=0.9999
    linearSpringConstant=300
    simulationSpace=BoneRelative
    alphaBoolBlend {blendOption=Linear}
    boundBone {boneName=dyn_spring_jnt_1}
    componentAppliedLinearAccClamp=<2000 2000 2000>
    componentLinearAccScale=<0.4 0.4 0.4>
    componentLinearVelScale=<0.4 0.4 0.4>
    physicsBodyDefinitions [{
      sphereCollisionRadius=0
      boundBone {boneName=dyn_spring_jnt_1}
      boxExtents=<1 1 1>
      constraintSetup {angularLimitsMax=<0 0 0>, angularLimitsMin=<0 0 0>, angularTarget=<0 0 0>, linearAxesMax=<5.7 0 0>, linearAxesMin=<-0.5 0 0>}
      localJointOffset=<0 0 0>
     }]
    relativeSpaceBone {boneName=Melee_TwoHanded_Handle}
   }
   n3 = Anim Dynamics - Bone: dyn_spring_jnt_2(ComponentPose=@n2, Alpha=1, ExternalForce="0,0,0") {
    bGravityOverrideInSimSpace=true
    bLinearSpring=true
    bOverrideLinearDamping=true
    gravityScale=0.4
    linearDampingOverride=0.9999
    linearSpringConstant=300
    simulationSpace=BoneRelative
    alphaBoolBlend {blendOption=Linear}
    boundBone {boneName=dyn_spring_jnt_2}
    componentAppliedLinearAccClamp=<2000 2000 2000>
    componentLinearAccScale=<0.4 0.4 0.4>
    componentLinearVelScale=<0.4 0.4 0.4>
    physicsBodyDefinitions [{
      sphereCollisionRadius=0
      boundBone {boneName=dyn_spring_jnt_2}
      boxExtents=<1 1 1>
      constraintSetup {angularLimitsMax=<0 0 0>, angularLimitsMin=<0 0 0>, angularTarget=<0 0 0>, linearAxesMax=<4.3 0 0>, linearAxesMin=<-0.5 0 0>}
      localJointOffset=<0 0 0>
     }]
    relativeSpaceBone {boneName=Melee_TwoHanded_Handle}
   }
   n4 = Anim Dynamics - Bone: dyn_spring_jnt_3(ComponentPose=@n3, Alpha=1, ExternalForce="0,0,0") {
    bGravityOverrideInSimSpace=true
    bLinearSpring=true
    bOverrideLinearDamping=true
    gravityScale=0.4
    linearDampingOverride=0.9999
    linearSpringConstant=300
    simulationSpace=BoneRelative
    alphaBoolBlend {blendOption=Linear}
    boundBone {boneName=dyn_spring_jnt_3}
    componentAppliedLinearAccClamp=<2000 2000 2000>
    componentLinearAccScale=<0.4 0.4 0.4>
    componentLinearVelScale=<0.4 0.4 0.4>
    physicsBodyDefinitions [{
      sphereCollisionRadius=0
      boundBone {boneName=dyn_spring_jnt_3}
      boxExtents=<1 1 1>
      constraintSetup {angularLimitsMax=<0 0 0>, angularLimitsMin=<0 0 0>, angularTarget=<0 0 0>, linearAxesMax=<2.9 0 0>, linearAxesMin=<-0.5 0 0>}
      localJointOffset=<0 0 0>
     }]
    relativeSpaceBone {boneName=Melee_TwoHanded_Handle}
   }
   n5 = Anim Dynamics - Bone: dyn_spring_jnt_4(ComponentPose=@n4, Alpha=1, ExternalForce="0,0,0") {
    bGravityOverrideInSimSpace=true
    bLinearSpring=true
    bOverrideLinearDamping=true
    gravityScale=0.4
    linearDampingOverride=0.9999
    linearSpringConstant=300
    simulationSpace=BoneRelative
    alphaBoolBlend {blendOption=Linear}
    boundBone {boneName=dyn_spring_jnt_4}
    componentAppliedLinearAccClamp=<2000 2000 2000>
    componentLinearAccScale=<0.4 0.4 0.4>
    componentLinearVelScale=<0.4 0.4 0.4>
    physicsBodyDefinitions [{
      sphereCollisionRadius=0
      boundBone {boneName=dyn_spring_jnt_4}
      boxExtents=<1 1 1>
      constraintSetup {angularLimitsMax=<0 0 0>, angularLimitsMin=<0 0 0>, angularTarget=<0 0 0>, linearAxesMax=<1.44 0 0>, linearAxesMin=<-0.5 0 0>}
      localJointOffset=<0 0 0>
     }]
    relativeSpaceBone {boneName=Melee_TwoHanded_Handle}
   }
   Save cached pose 'Coil'(Pose=Component To Local(ComponentPose=@n5)) {cacheName=Coil}
   n6 = "Transform (Modify) Bone - Bone: dyn_R_hammerSpring_jnt"(
    ComponentPose=Local To Component(LocalPose=Use cached pose 'Coil'())
    Alpha=Get Hammerhead Alpha Squash() -> (Hammerhead Alpha Squash:double)
    Translation="0,0,0"
    Rotation="0,0,0"
    Scale="0.2,1,1"
   ) {
    scaleMode=BMM_Replace
    alphaBoolBlend {blendOption=Linear}
    alphaScaleBiasClamp {bInterpResult=true}
    boneToModify {boneName=dyn_R_hammerSpring_jnt}
    scale {x=0.2}
   }
   Save cached pose 'HammerSquash'(Pose=Component To Local(ComponentPose=@n6)) {cacheName=HammerSquash}
  ]}
 EventGraph {nodes [
   Event BlueprintUpdateAnimation()
   > n1 = Cast To B_Melee_Impact_Pickaxe_Athena_WickedDuck!(execute, Object=GetOwningActor()) -> (then, CastFailed, AsB Melee Impact Pickaxe Athena Wicked Duck:B Melee Impact Pickaxe Athena Wicked Duck Object Reference) {is=title K2Node_DynamicCast}
   > Branch(Condition=Get SuccessfulHit!(self:B Melee Impact Pickaxe Athena Wicked Duck Object Reference=@n1) -> (SuccessfulHit:bool))
   > Sequence()
   > Set Hammerhead Alpha Squash(Hammerhead Alpha Squash:double=1) -> (Output_Get:double)
   > Delay(Duration=0.3)
   > Set Hammerhead Alpha Squash(Hammerhead Alpha Squash=0)
   Set Hammerhead Alpha Stretch(Hammerhead Alpha Stretch:double=0) -> (Output_Get:double)
   > Delay(Duration=0.4)
   > Set Hammerhead Alpha Stretch(Hammerhead Alpha Stretch=1)
   > Delay(Duration=0.4)
   > Set Hammerhead Alpha Stretch(Hammerhead Alpha Stretch=0)
  ]}
]
variables [
 Hammerhead Alpha Squash {type=number}
 Hammerhead Alpha Stretch {type=number}
]
